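# Sound cue that plays the looping music for the Cyclone Headbang emote, set up as the first-person (1P) emote music variant.
# A single looping wave player feeds the root directly, with boosted volume and no other nodes.
# It routes through the 1P emote music sound class, uses the Athena emote music 1P concurrency set, and sends to the music audio-analysis submix.
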SoundCue /BRCosmetics/Sounds/Emotes/Cyclone/Emote_Cyclone_Headbang_Cue
root=WavePlayer_0
nodes [{id=WavePlayer_0, properties {bLooping=true, soundWaveAssetPtr=&/Cyclone_Headbang_Loop}}]
properties {
 maxDistance=2097152
 soundClassObject=/Game/Sounds/FortSoundClassesAndModes/Fort_Music_Emote_1P_PSM
 volumeMultiplier=1.5
 concurrencyOverrides {resolutionRule=PreventNew}
 concurrencySet [/Game/Sounds/Concurrency_Rules/Athena_EmoteMusic_1P_Conc]
 soundSubmixSends [{minSendLevel=0, soundSubmix=&Fortnite_AudioAnalysis_Music}]
}
